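; Projectile Blueprint for the LumpyRug mythic item, a homing rocket that explodes when it stops, enters water or reaches its maximum lifetime.
; An optional timer delays enabling stop-on-water behavior, then detonates the rocket if it is already submerged.
; DoOnce and an exploded flag guard against repeat detonations.
; Homing switches off once the rocket passes its target or reaches the lock distance threshold.
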(event OnEnterWaterEvent(WaterInteractionComponent)
  (params (WaterBody :type Fort%20Water%20Body%20Actor%20Object%20Reference) (WaterInteractionComponent :type Fort%20Water%20Interaction%20Component%20Object%20Reference) (bIsFirstBody :type Boolean))
  ;[1 1 1] Explode
  (Utilities|FlowControl|DoOnce false
    (:Completed
      (Delivery|DoExplosion) ; Explosion
      (Variables|VFX|SetbHasExploded? true))))

(event EventOnStop
  (params (Hit :type Hit%20Result%20Structure%20%28by%20ref%29))
  ;[1 1 1] Explode
  (Utilities|FlowControl|DoOnce false
    (:Completed
      (Delivery|DoExplosion) ; Explosion
      (Variables|VFX|SetbHasExploded? true))))

(event Custom|RocketWaterDelay
  ;[1 1 1] Explode
  (if (and (Utilities|IsValid (Variables|/Script/Engine.MovementComponent|GetUpdatedComponent (Variables|Movement|GetProjectileMovementComponent)) :!class /Script/Engine.KismetSystemLibrary) (not (Delivery|IsBeingKilled)))
    (Variables|Projectile|SetTouchWaterBehavior "StopIfStopSimulatingOnHit")
    (if (Water|IsinWater (Variables|Water|GetWaterInteractionComponent))
      (Utilities|FlowControl|DoOnce false
        (:Completed
          (Delivery|DoExplosion) ; Explosion
          (Variables|VFX|SetbHasExploded? true))))))

(event EventBeginPlay
  (Utilities|FlowControl|Sequence
    (:then_0
      (bind _returnvalue (Utilities|Time|SetTimerbyEvent (delegate Custom|Timeout) (Variables|Default|GetProjectile%20Max%20Lifetime)))
      (Variables|Default|SetLifespan%20Timer _returnvalue))
    (:then_1
      (if (> (Variables|Default|GetWaterInteractionDelay) 0.0)
        (Utilities|Time|SetTimerbyEvent (EventDispatchers|CreateEvent :!event Custom|RocketWaterDelay) (Variables|Default|GetWaterInteractionDelay))
        (else
          (Variables|Projectile|SetTouchWaterBehavior "StopIfStopSimulatingOnHit"))))))

(event Custom|Timeout
  (Components|Movement|StopMovementImmediately (Variables|Movement|GetProjectileMovementComponent))
  (if (not (Variables|VFX|GetbHasExploded?))
    (Variables|VFX|SetbHasExploded? true)
    (Delivery|DoExplosion) ; Explosion
  ))

(event ProjectileMovement|EventOnHomingPassedTarget
  (Projectile|SetHomingStyle 0 0 "None" false :!class /Script/FortniteGame.FortProjectileMovementInterface))

(event ProjectileMovement|EventOnReachedLockTargetDistanceThreshold
  (Projectile|SetHomingStyle 0 0 "None" false :!class /Script/FortniteGame.FortProjectileMovementInterface))

(fn ConstructionScript)
