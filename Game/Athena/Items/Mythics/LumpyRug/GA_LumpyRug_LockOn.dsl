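; Gameplay ability that drives target lock-on for the LumpyRug rocket launcher.
; A fast looping timer polls for an aimed-at enemy, building tracking progress while the target stays the same and resetting it otherwise.
; Full progress sets the weapon's lock-on target, applies a HasTarget effect and plays a lock sound.
; The weapon's widget shows target, lock state and progress through a material parameter; losing the target or ending the ability clears the lock.
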(event Ability|EventActivateAbility
  (bind _asfort_player_pawn (Utilities|Casting|CastToFortPlayerPawn (Ability|GetAvatarActorfromActorInfo)))
  (Variables|Default|SetPlayerPawn _asfort_player_pawn)
  (Ability|Tasks|WaitInputRelease false)
  (bind _returnvalue (Utilities|Time|SetTimerbyEvent (delegate Custom|TrackingTick) 0.01 true)))

(event Custom|ResetSoundCue
  (Utilities|FlowControl|Sequence
    (:then_0
      (Utilities|IsValid (stale _returnvalue)
        (:Is%20Valid
          (Components|DestroyComponent (stale _returnvalue)))))
    (:then_1
      (Utilities|FlowControl|DoOnce false
        (:Completed
          (bind _returnvalue (Audio|SpawnSound2D "/Game/Sounds/Fort_Weapons/Launchers/LumpyRug/Athena_LumpyRug_LockedOn_SoundCue.Athena_LumpyRug_LockedOn_SoundCue")))))))

(event Custom|TrackingTick
  (bind _currentweapon (Variables|Default|GetCurrentWeapon))
  (bind _lockonwidget (Variables|/Game/Weapons/FORT_RocketLaunchers/Blueprints/B_LumpyRug_Athena.B_LumpyRug_Athena_C|GetLockOnWidget _currentweapon))
  (bind _widget (Variables|Default|GetWidget))
  (bind _widget_2 (Variables|Default|GetWidget))
  (bind _widget_3 (Variables|Default|GetWidget))
  (bind _returnvalue (Fortnite|GetAimedAtEnemy (Fortnite|GetFortPlayerControllerfromActor (Ability|GetAvatarActorfromActorInfo)) 7e+02 7e+02 2e+04 true))
  (bind _asb_lumpy_rug_athena (Utilities|Casting|CastToB_LumpyRug_Athena (Variables|/Script/FortniteGame.FortPawn|GetCurrentWeapon (Variables|Default|GetPlayerPawn))))
  (bind _output_get (Variables|Default|SetCurrentWeapon _asb_lumpy_rug_athena))
  (Utilities|IsValid (Variables|/Game/Weapons/FORT_RocketLaunchers/Blueprints/B_LumpyRug_Athena.B_LumpyRug_Athena_C|GetLockOnWidget _output_get)
    (:Is%20Valid
      (Variables|Default|SetWidget (Variables|/Game/Weapons/FORT_RocketLaunchers/Blueprints/B_LumpyRug_Athena.B_LumpyRug_Athena_C|GetLockOnWidget _output_get)))
    (:Is%20Not%20Valid
      _))
  (Utilities|IsValid _returnvalue
    (:Is%20Valid
      (Widget|SetVisibility _widget_3 "Visible")
      (Variables|/Game/Weapons/FORT_RocketLaunchers/Blueprints/B_LumpyRug_Athena.B_LumpyRug_Athena_C|SetPotentialTarget _returnvalue _output_get)
      (if (Utilities|Equal(Object) _returnvalue (Variables|Default|GetTrackedTarget))
        (Variables|Default|SetTracking (Math|Float|Clamp(Float) (+ (Variables|Default|GetTracking) 0.025)))
        (else
          (Variables|Default|SetTracking 0.0)
          (CallFunction|ResetSoundCue)
          (Utilities|IsValid (Variables|Default|GetWidget)
            (:Is%20Valid
              (Variables|/Game/Athena/Items/Mythics/LumpyRug/Athena_LumpyRug_LockOnWidget.Athena_LumpyRug_LockOnWidget_C|SetLockedOn false (Variables|/Game/Weapons/FORT_RocketLaunchers/Blueprints/B_LumpyRug_Athena.B_LumpyRug_Athena_C|GetLockOnWidget (Variables|Default|GetCurrentWeapon))))
            (:Is%20Not%20Valid
              _))))
      (Variables|Default|SetTrackedTarget _returnvalue)
      (Utilities|FlowControl|Sequence
        (:then_0
          (Utilities|IsValid _widget
            (:Is%20Valid
              (Variables|/Game/Athena/Items/Mythics/LumpyRug/Athena_LumpyRug_LockOnWidget.Athena_LumpyRug_LockOnWidget_C|SetTarget (Variables|Default|GetTrackedTarget) _widget))
            (:Is%20Not%20Valid
              _))
          (if (>= (Variables|Default|GetTracking) 1.0)
            (Utilities|IsValid _widget_2
              (:Is%20Valid
                (Variables|/Game/Athena/Items/Mythics/LumpyRug/Athena_LumpyRug_LockOnWidget.Athena_LumpyRug_LockOnWidget_C|SetLockedOn true _widget_2))
              (:Is%20Not%20Valid
                _))
            (bind _returnvalue_2 (Ability|ApplyGameplayEffectToOwner "/Game/Athena/Items/Mythics/LumpyRug/GE_Athena_LumpyRug_HasTarget.GE_Athena_LumpyRug_HasTarget_C"))
            (Variables|/Game/Weapons/FORT_RocketLaunchers/Blueprints/B_LumpyRug_Athena.B_LumpyRug_Athena_C|SetLockOnTarget (Variables|Default|GetTrackedTarget) (Variables|Default|GetCurrentWeapon))
            (Utilities|FlowControl|DoOnce false
              (:Completed
                (bind _returnvalue_3 (Audio|SpawnSound2D "/Game/Sounds/Fort_Weapons/Launchers/LumpyRug/Athena_LumpyRug_LockedOn_SoundCue.Athena_LumpyRug_LockedOn_SoundCue"))))))
        (:then_1
          _)))
    (:Is%20Not%20Valid
      (Widget|SetVisibility _widget_3 "Hidden")
      (Variables|/Game/Weapons/FORT_RocketLaunchers/Blueprints/B_LumpyRug_Athena.B_LumpyRug_Athena_C|SetPotentialTarget 0 _output_get)
      (Variables|Default|SetTracking 0.0)
      (Variables|/Game/Weapons/FORT_RocketLaunchers/Blueprints/B_LumpyRug_Athena.B_LumpyRug_Athena_C|SetLockOnTarget 0 (Variables|Default|GetCurrentWeapon))
      (Utilities|IsValid (Variables|Default|GetWidget)
        (:Is%20Valid
          (Variables|/Game/Athena/Items/Mythics/LumpyRug/Athena_LumpyRug_LockOnWidget.Athena_LumpyRug_LockOnWidget_C|SetTarget 0 _lockonwidget)
          (Variables|/Game/Athena/Items/Mythics/LumpyRug/Athena_LumpyRug_LockOnWidget.Athena_LumpyRug_LockOnWidget_C|SetLockedOn false _lockonwidget))
        (:Is%20Not%20Valid
          _))
      (CallFunction|ResetSoundCue)
      (Ability|RemoveGameplayEffectFromOwnerWithHandle _returnvalue_2)))
  (Utilities|IsValid (Variables|Default|GetWidget)
    (:Is%20Valid
      (bind _returnvalue_4 (Appearance|GetDynamicMaterial (Variables|/Game/Athena/Items/Mythics/LumpyRug/Athena_LumpyRug_LockOnWidget.Athena_LumpyRug_LockOnWidget_C|GetImage_83 (Variables|Default|GetWidget)) :!class /Script/UMG.Image))
      (Rendering|Material|SetScalarParameterValue _returnvalue_4 "CenterSize" (Variables|Default|GetTracking) :!class /Script/Engine.MaterialInstanceDynamic))))

(event Ability|EventOnEndAbility
  (params (bWasCancelled :type Boolean))
  (bind _currentweapon (Variables|Default|GetCurrentWeapon))
  (bind _lockonwidget (Variables|/Game/Weapons/FORT_RocketLaunchers/Blueprints/B_LumpyRug_Athena.B_LumpyRug_Athena_C|GetLockOnWidget _currentweapon))
  (Utilities|Time|ClearandInvalidateTimerbyHandle (stale Ability|EventActivateAbility _returnvalue))
  (Widget|SetVisibility (Variables|Default|GetWidget) "Hidden")
  (Variables|Default|SetTracking 0.0)
  (Variables|/Game/Weapons/FORT_RocketLaunchers/Blueprints/B_LumpyRug_Athena.B_LumpyRug_Athena_C|SetLockOnTarget 0 (Variables|Default|GetCurrentWeapon))
  (Utilities|IsValid (Variables|Default|GetWidget)
    (:Is%20Valid
      (Variables|/Game/Athena/Items/Mythics/LumpyRug/Athena_LumpyRug_LockOnWidget.Athena_LumpyRug_LockOnWidget_C|SetTarget 0 _lockonwidget)
      (Variables|/Game/Athena/Items/Mythics/LumpyRug/Athena_LumpyRug_LockOnWidget.Athena_LumpyRug_LockOnWidget_C|SetLockedOn false _lockonwidget))
    (:Is%20Not%20Valid
      _))
  (CallFunction|ResetSoundCue)
  (Ability|RemoveGameplayEffectFromOwnerWithHandle (stale Custom|TrackingTick _returnvalue_2))
  (Utilities|IsValid (Variables|Default|GetWidget)
    (:Is%20Valid
      (bind _returnvalue (Appearance|GetDynamicMaterial (Variables|/Game/Athena/Items/Mythics/LumpyRug/Athena_LumpyRug_LockOnWidget.Athena_LumpyRug_LockOnWidget_C|GetImage_83 (Variables|Default|GetWidget)) :!class /Script/UMG.Image))
      (Rendering|Material|SetScalarParameterValue _returnvalue "CenterSize" (Variables|Default|GetTracking) :!class /Script/Engine.MaterialInstanceDynamic))))

(event Ability|EventFailedToActivatePassiveAbility
  (params (ActorInfo :type Gameplay%20Ability%20Actor%20Info%20Structure)))
