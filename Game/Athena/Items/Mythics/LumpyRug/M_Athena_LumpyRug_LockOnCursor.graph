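# Translucent UI material that draws a procedural lock-on cursor for the LumpyRug mythic item.
# Scaled square masks are subtracted to form a center outline and a spinning outer border, with a rotated square cutting that border into brackets.
# Eased scalar lerps animate size, gap, cut, spin, color and opacity, with border thickness converted through UI pixel size.
# A tiled, vertically panning gradient modulates the combined mask like scanlines.
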Material /Game/Athena/Items/Mythics/LumpyRug/M_Athena_LumpyRug_LockOnCursor
graph {
 properties {blendMode=BLEND_Translucent, materialDomain=MD_UI, numCustomizedUVs=1, pixelDepthOffsetMode=PDOM_Legacy, usedShadingModels=Default Lit}
 nodes [
  n1 = LinearInterpolate(
   A=ScalarParameter() {defaultValue=1.97, parameterName=CenterSizeStart}
   B=ScalarParameter() {defaultValue=1.97, parameterName=CenterSizeEnd}
   Alpha=Power(Base=ScalarParameter() {defaultValue=0.5, sliderMax=1}, Exp=ScalarParameter() {defaultValue=1, parameterName=CenterSizeEasePower})
  )
  n2 = Add(A=@n1, B=ScalarParameter() {defaultValue=0.4, parameterName=OuterBorderGap})
  n3 = LinearInterpolate(
   B=@n2
   Alpha=Power(Base=ScalarParameter() {defaultValue=0.5, sliderMax=1}, Exp=ScalarParameter() {defaultValue=1, parameterName=BorderGapEase})
  ) {constA=1, constB=0.5}
  n4 = ScalarParameter() {defaultValue=0.5, sliderMax=1}
  n5 = ScalarParameter() {defaultValue=1, parameterName=BorderGapEase}
  n6 = Clamp(None=LinearInterpolate(A=ScalarParameter() {parameterName=LockOnShapeCut}, B=ScalarParameter() {parameterName=LockOnShapeCutEnd}, Alpha=Power(Base=@n4, Exp=@n5)) {constA=1, constB=0})
  Add(A=@n3, B=@n6)
  LinearInterpolate(B=@n2, Alpha=ScalarParameter() {defaultValue=0.5, sliderMax=1}) {constA=1, constB=0.5}
  n7 = VectorParameter() {parameterName=CoreColor, defaultValue=<rgba 1 1 1 1>}
  n8 = TextureCoordinate()
  n9 = LinearInterpolate(
   A=ScalarParameter() {defaultValue=1, parameterName=SpinAmount}
   Alpha=Power(Base=ScalarParameter() {defaultValue=0.5, sliderMax=1}, Exp=ScalarParameter() {defaultValue=1, parameterName=SpinEasePower})
  ) {constA=3, constB=0}
  n10 = MaterialFunctionCall(UVs=@n8, "Rotation Angle (0-1)"=@n9) {materialFunction=&CustomRotator}
  n11 = MaterialFunctionCall(UVs=@n10.Rotated Values, Texture Scale=@n3) {materialFunction=&ScaleUVsByCenter}
  n12 = LinearInterpolate(
   A=ScalarParameter() {defaultValue=4, parameterName=BorderSizeStart}
   B=ScalarParameter() {defaultValue=4, parameterName=BorderSizeEnd}
   Alpha=ScalarParameter() {defaultValue=0.5, sliderMax=1}
  )
  n13 = MaterialFunctionCall() {materialFunction=&GetUserInterfaceUV}
  n14 = MaterialFunctionCall(UVs=@n10.Rotated Values, Texture Scale=Subtract(A=@n3, B=Divide(A=@n12, B=@n13.Pixel Size))) {materialFunction=&ScaleUVsByCenter}
  n15 = MaterialFunctionCall(UVs=@n8, "Rotation Angle (0-1)"=Add(A=@n9, B=ScalarParameter() {defaultValue=0.125, parameterName=LockOnShapeRotation})) {materialFunction=&CustomRotator}
  n16 = Clamp(None=Subtract(
    A=Subtract(A=@n11.0-1 mask, B=@n14.0-1 mask)
    B=MaterialFunctionCall(UVs=@n15.Rotated Values, Texture Scale=@n6) {materialFunction=&ScaleUVsByCenter}.0-1 mask
   ))
  LinearInterpolate(
   A=LinearInterpolate(A=@n7.RGB, B=VectorParameter() {parameterName=BracketColor, defaultValue=<rgba 1 0.3403 0 1>}.RGB, Alpha=@n16)
   B=@n7.RGB
   Alpha=Power(Base=ScalarParameter() {defaultValue=0.5, sliderMax=1}, Exp=ScalarParameter() {defaultValue=1, parameterName=ColorEasePower})
  )
  MaterialFunctionCall(UVs=TextureCoordinate(), "Rotation Angle (0-1)"=ScalarParameter() {defaultValue=0.125, parameterName=DefaultRotation}) {materialFunction=&CustomRotator}
  n17 = ScalarParameter() {defaultValue=0.5, sliderMax=1}
  n18 = ScalarParameter() {defaultValue=1, parameterName=CenterOpacityEasePower}
  n19 = TextureCoordinate()
  n20 = Subtract(
   A=@n1
   B=Divide(
    A=ScalarParameter() {defaultValue=4, parameterName=BorderSize}
    B=MaterialFunctionCall() {materialFunction=&GetUserInterfaceUV}.Pixel Size
   )
  )
  n21 = Multiply(
   A=Subtract(
    A=MaterialFunctionCall(UVs=@n19, Texture Scale=@n1) {materialFunction=&ScaleUVsByCenter}.0-1 mask
    B=MaterialFunctionCall(UVs=@n19, Texture Scale=@n20) {materialFunction=&ScaleUVsByCenter}.0-1 mask
   )
   B=LinearInterpolate(Alpha=Power(Base=@n17, Exp=@n18))
  )
  n22 = ScalarParameter() {defaultValue=0.5, sliderMax=1}
  n23 = ScalarParameter() {defaultValue=1, parameterName=BracketOpacityEasePower}
  n24 = Multiply(A=@n16, B=LinearInterpolate(Alpha=Power(Base=@n22, Exp=@n23)))
  n25 = TextureSampleParameter2D(UVs=Panner(Coordinate=TextureCoordinate() {vTiling=22}) {speedY=0.25}) {parameterName=Panner, samplerType=SAMPLERTYPE_LinearColor, texture=/Game/Global/Textures/Default/Masks_and_Gradients/T_VertGradient}
  Multiply(
   A=Multiply(A=Add(A=@n21, B=@n24), B=ScalarParameter() {defaultValue=1, parameterName=CenterOpacity})
   B=Clamp(None=Multiply(A=@n25.RGB) {constB=1.25}) {maxDefault=2, minDefault=0.8}
  )
  ScalarParameter() {defaultValue=1, parameterName=EasePower}
 ]
}
